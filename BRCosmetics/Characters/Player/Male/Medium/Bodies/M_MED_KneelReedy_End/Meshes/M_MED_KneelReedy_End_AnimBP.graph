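# Post-body AnimBP for the KneelReedy End cosmetic skin, layering Control Rig and additive cape-attach offsets over the parent mesh's pose.
# Overlay-override events swap outline and emote materials between End and base variants, chosen by mesh names.
# Updates copy brow flipbook curves to the head overlay and, while MaterialCurveA is active, gadget ring transforms to materials.
# LOD-driven skin weight profiles toggle retargeting; arm-visibility events set the rig's gauntlet flag.
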AnimBlueprint /BRCosmetics/Characters/Player/Male/Medium/Bodies/M_MED_KneelReedy_End/Meshes/M_MED_KneelReedy_End_AnimBP
target_skeleton=&/M_MED_KneelReedy_End_Skeleton
graphs [
 AnimGraph {nodes [
   n1 = Control Rig!(
    Source=Copy Pose From Mesh() {bCopyCurves=true, bUseAttachedParent=true}
    Alpha
    disable retargeting:bool=Get Disable Retargeting() -> (Disable Retargeting:bool)
    isWearingGauntlets:bool=Get isWearingGauntlets() -> (isWearingGauntlets:bool)
   ) -> (Pose) {controlRigAssetReference {blueprintRigClass=&/M_MED_KneelReedy_End_CtrlRig.M_MED_KneelReedy_End_CtrlRig_C}}
   n2 = "Transform (Modify) Bone - Bone: attach_cape"!(ComponentPose=Local To Component(LocalPose=@n1), bAlphaBoolEnabled:bool=True, Translation="0,-1,0", Rotation="0,0,12") -> (Pose) {
    alphaInputType=Bool
    rotationMode=BMM_Additive
    rotationSpace=BCS_BoneSpace
    translationMode=BMM_Additive
    translationSpace=BCS_BoneSpace
    alphaBoolBlend {blendInTime=0.5, blendOption=CircularOut, blendOutTime=0.5}
    boneToModify {boneName=attach_cape}
    rotation {roll=12}
    translation {y=-1}
   }
   n3 = "Transform (Modify) Bone - Bone: attach_cape"(ComponentPose=@n2, bAlphaBoolEnabled=True, Translation="0,-2,0", Rotation="0,0,15") {
    alphaInputType=Bool
    rotationMode=BMM_Additive
    rotationSpace=BCS_BoneSpace
    translationMode=BMM_Additive
    translationSpace=BCS_BoneSpace
    alphaBoolBlend {blendInTime=0.5, blendOption=CircularOut, blendOutTime=0.5}
    boneToModify {boneName=attach_cape}
    rotation {roll=15}
    translation {y=-2}
   }
   Output Pose(Result=Component To Local(ComponentPose=@n3)) {layerGroup=None}
  ]}
 EventGraph {nodes [
   OnStartOverlayOverride() {is=title K2Node_CustomEvent}
   n1 = GetOwningActor()
   > n2 = Cast To FortPlayerPawn(Object=@n1)
   n3 = GetSkeletalMeshForPartType(self=@n2)
   n4 = GetSkeletalMeshForPartType(self=@n2, PartType=Head)
   n5 = GetSkeletalMeshForPartType(self=@n2, PartType=Face)
   n6 = OR Boolean!(
    A=Contains(SearchIn=GetDisplayName(Object=@n4), Substring=End)
    B=Contains(SearchIn=GetDisplayName(Object=@n5), Substring=End)
    C:bool=Contains(SearchIn=GetDisplayName(Object=@n3), Substring=End)
   ) -> (ReturnValue)
   > n7 = Branch(Condition=@n6)
   > SetOverlayMaterial!(
    execute
    self:Mesh Component Object Reference=@n4
    NewOverlayMaterial:Material Interface Object Reference=/BRCosmetics/Characters/Player/Male/Medium/Bodies/M_MED_KneelReedy_End/Materials/Outlines/MI_KneelReedy_End_Head_Outline_Appear
    bSetMaterialSlot:bool=false
    SlotIndex:int=0
   ) -> (then)
   > SetOverlayMaterial(
    self=@n5
    NewOverlayMaterial=/BRCosmetics/Characters/Player/Male/Medium/Bodies/M_MED_KneelReedy_End/Materials/Outlines/MI_KneelReedy_End_FaceAcc_Outline_Appear
    bSetMaterialSlot=false
    SlotIndex=0
   )
   > SetOverlayMaterial(
    self=@n3
    NewOverlayMaterial=/BRCosmetics/Characters/Player/Male/Medium/Bodies/M_MED_KneelReedy_End/Materials/Outlines/MI_KneelReedy_End_Body_Outline_Appear
    bSetMaterialSlot=false
    SlotIndex=0
   )
   > SetMaterial(
    self=@n4
    ElementIndex=0
    Material=/BRCosmetics/Characters/Player/Male/Medium/Bodies/M_MED_KneelReedy_End/Materials/MI_KneelReedy_End_Head_Emote
   )
   > SetMaterial(
    self=@n5
    ElementIndex=0
    Material=/BRCosmetics/Characters/Player/Male/Medium/Bodies/M_MED_KneelReedy_End/Materials/MI_KneelReedy_End_FaceAcc_Emote
   )
   > SetMaterial(
    self=@n3
    ElementIndex=0
    Material=/BRCosmetics/Characters/Player/Male/Medium/Bodies/M_MED_KneelReedy_End/Materials/MI_KneelReedy_End_Body_Emote
   )
   SetOverlayMaterial(
    execute=@n7.else
    self=@n4
    NewOverlayMaterial=/BRCosmetics/Characters/Player/Male/Medium/Bodies/M_MED_KneelReedy/Materials/Default/MI_KneelReedy_Head_Outline_Appear
    bSetMaterialSlot=false
    SlotIndex=0
   )
   > SetOverlayMaterial(
    self=@n5
    NewOverlayMaterial=/BRCosmetics/Characters/Player/Male/Medium/Bodies/M_MED_KneelReedy/Materials/Default/MI_KneelReedy_FaceAcc_Outline_Appear
    bSetMaterialSlot=false
    SlotIndex=0
   )
   > SetOverlayMaterial(
    self=@n3
    NewOverlayMaterial=/BRCosmetics/Characters/Player/Male/Medium/Bodies/M_MED_KneelReedy/Materials/Default/MI_KneelReedy_Body_Outline_Appear
    bSetMaterialSlot=false
    SlotIndex=0
   )
   > SetMaterial(self=@n4, ElementIndex=0, Material=/BRCosmetics/Characters/Player/Male/Medium/Bodies/M_MED_KneelReedy/Materials/MI_KneelReedy_Head_Emote)
   > SetMaterial(self=@n5, ElementIndex=0, Material=/BRCosmetics/Characters/Player/Male/Medium/Bodies/M_MED_KneelReedy/Materials/MI_KneelReedy_FaceAcc_Emote)
   > SetMaterial(self=@n3, ElementIndex=0, Material=/BRCosmetics/Characters/Player/Male/Medium/Bodies/M_MED_KneelReedy/Materials/MI_KneelReedy_Body_Emote)
   OnEndOverlayOverride() {is=title K2Node_CustomEvent}
   > n8 = Cast To FortPlayerPawn(Object=@n1)
   n9 = GetSkeletalMeshForPartType(self=@n8, PartType=Head)
   n10 = GetSkeletalMeshForPartType(self=@n8, PartType=Face)
   n11 = GetSkeletalMeshForPartType(self=@n8)
   n12 = OR Boolean(
    A=Contains(SearchIn=GetDisplayName(Object=@n9), Substring=End)
    B=Contains(SearchIn=GetDisplayName(Object=@n10), Substring=End)
    C=Contains(SearchIn=GetDisplayName(Object=@n11), Substring=End)
   )
   > n13 = Branch(Condition=@n12)
   > SetOverlayMaterial(
    self=@n9
    NewOverlayMaterial=/BRCosmetics/Characters/Player/Male/Medium/Bodies/M_MED_KneelReedy_End/Materials/Outlines/MI_KneelReedy_End_Head_Outline
    bSetMaterialSlot=false
    SlotIndex=0
   )
   > SetOverlayMaterial(
    self=@n10
    NewOverlayMaterial=/BRCosmetics/Characters/Player/Male/Medium/Bodies/M_MED_KneelReedy_End/Materials/Outlines/MI_KneelReedy_End_FaceAcc_Outline
    bSetMaterialSlot=false
    SlotIndex=0
   )
   > SetOverlayMaterial(
    self=@n11
    NewOverlayMaterial=/BRCosmetics/Characters/Player/Male/Medium/Bodies/M_MED_KneelReedy_End/Materials/Outlines/MI_KneelReedy_End_Body_Outline
    bSetMaterialSlot=false
    SlotIndex=0
   )
   > SetMaterial(self=@n9, ElementIndex=0, Material=/BRCosmetics/Characters/Player/Male/Medium/Bodies/M_MED_KneelReedy_End/Materials/MI_KneelReedy_End_Head)
   > SetMaterial(
    self=@n10
    ElementIndex=0
    Material=/BRCosmetics/Characters/Player/Male/Medium/Bodies/M_MED_KneelReedy_End/Materials/MI_KneelReedy_End_FaceAcc
   )
   > SetMaterial(self=@n11, ElementIndex=0, Material=/BRCosmetics/Characters/Player/Male/Medium/Bodies/M_MED_KneelReedy_End/Materials/MI_KneelReedy_End_Body)
   SetOverlayMaterial(
    execute=@n13.else
    self=@n9
    NewOverlayMaterial=/BRCosmetics/Characters/Player/Male/Medium/Bodies/M_MED_KneelReedy/Materials/Default/MI_KneelReedy_Head_Outline
    bSetMaterialSlot=false
    SlotIndex=0
   )
   > SetOverlayMaterial(
    self=@n10
    NewOverlayMaterial=/BRCosmetics/Characters/Player/Male/Medium/Bodies/M_MED_KneelReedy/Materials/Default/MI_KneelReedy_FaceAcc_Outline
    bSetMaterialSlot=false
    SlotIndex=0
   )
   > SetOverlayMaterial(
    self=@n11
    NewOverlayMaterial=/BRCosmetics/Characters/Player/Male/Medium/Bodies/M_MED_KneelReedy/Materials/Default/MI_KneelReedy_Body_Outline
    bSetMaterialSlot=false
    SlotIndex=0
   )
   > SetMaterial(self=@n9, ElementIndex=0, Material=/BRCosmetics/Characters/Player/Male/Medium/Bodies/M_MED_KneelReedy/Materials/Default/MI_KneelReedy_Head)
   > SetMaterial(
    self=@n10
    ElementIndex=0
    Material=/BRCosmetics/Characters/Player/Male/Medium/Bodies/M_MED_KneelReedy/Materials/Default/MI_KneelReedy_FaceAcc
   )
   > SetMaterial(self=@n11, ElementIndex=0, Material=/BRCosmetics/Characters/Player/Male/Medium/Bodies/M_MED_KneelReedy/Materials/Default/MI_KneelReedy_Body)
   n14 = Cast To FortPlayerPawn(Object=GetOwningActor())
   n15 = GetMaterial!(self:Primitive Component Object Reference=GetSkeletalMeshForPartType(self=@n14), ElementIndex:int=0) -> (ReturnValue:Material Interface Object Reference)
   > n16 = Branch(Condition="Not Equal (Object)"(A:Object Reference=Get Previous Mat() -> (Previous Mat:Material Interface Object Reference), B:Object Reference=@n15) -> (ReturnValue:bool=~) {is=title K2Node_PromotableOperator})
   > n17 = Set Previous Mat(Previous Mat:Material Interface Object Reference=@n15) -> (Output_Get:Material Interface Object Reference)
   n18 = Sequence!(execute=[@n39.then, @n38.then, @n17.then]) -> (then_0, then_1, then_2:Exec, then_3:Exec)
   > n19 = Cast To FortPlayerPawn(Object=GetOwningActor())
   n20 = GetSkeletalMeshForPartType(self=@n19, PartType=Head)
   n21 = GetAnimInstance(self=@n20)
   > SetScalarParameterValueOnOverlayMaterial!(
    execute
    MeshComponent:Mesh Component Object Reference=@n20
    ParameterName:Name=Flipbook_face_L_brow_index
    ParameterValue:float=GetCurveValue(self=@n21, CurveName=Flipbook_face_L_brow_index)
   ) -> (then)
   > SetScalarParameterValueOnOverlayMaterial(MeshComponent=@n20, ParameterName=Flipbook_face_R_brow_index, ParameterValue=GetCurveValue(self=@n21, CurveName=Flipbook_face_R_brow_index))
   n22 = GetAnimInstance(self=GetPartSkeletalMeshComponent())
   > Set Active(Active:bool="float > float"(A=GetCurveValue(self=@n22, CurveName=MaterialCurveA))) -> (Output_Get:bool)
   n23 = Branch(execute=@n18.then_1, Condition="integer >= integer"(A=Get CurrentLOD(), B=2))
   > n24 = Sequence!(execute) -> (then_0, then_1)
   > Do Once(Reset=@n26.then_0)
   n25 = GetPartSkeletalMeshComponent()
   > SetSkinWeightProfile!(execute, self:Skinned Mesh Component Object Reference=@n25, InProfileName:Name=Alternative, InLayer:ESkinWeightProfileLayer Enum=Secondary) -> (then, ReturnValue:bool)
   Do Once(execute=@n26.then_1, Reset=@n24.then_1)
   > ClearSkinWeightProfile!(execute, self:Skinned Mesh Component Object Reference=@n25, InLayer:ESkinWeightProfileLayer Enum=Secondary) -> (then)
   n26 = Sequence(execute=@n23.else)
   Branch(execute=@n18.then_3, Condition=Get Active() -> (Active:bool))
   n27 = Get GadgetComponent() -> (GadgetComponent:Skeletal Mesh Component Object Reference)
   > n28 = Is Valid(InputObject=@n27)
   n29 = GetOwningActor()
   n30 = Get Components By Class!(self:Actor Object Reference=@n29, ComponentClass:Actor Component Class Reference=/Script/Engine.SkeletalMeshComponent) -> (ReturnValue:Array of Skeletal Mesh Component Object References)
   n31 = For Each Loop(Exec=@n28.Is Not Valid, Array:Array of Skeletal Mesh Component Object References=@n30) -> (Array Element:Skeletal Mesh Component Object Reference)
   > Branch(Condition="Equal (Object)"(
     A=Get SkeletalMesh!(self:Skinned Mesh Component Object Reference=@n31) -> (SkeletalMesh:Skeletal Mesh Object Reference)
     B=/BRCosmetics/Animation/Props/Gadgets_KneelReedy/Meshes/KneelReedy_Gadget
    ))
   > Set GadgetComponent(GadgetComponent:Skeletal Mesh Component Object Reference=@n31) -> (Output_Get:Skeletal Mesh Component Object Reference)
   AddTickPrerequisiteComponent!(execute=@n28.Is Valid, self:Actor Object Reference=@n29, PrerequisiteComponent:Actor Component Object Reference=@n27) -> (then)
   > n32 = For Each Loop(Array=@n30)
   n33 = GetBoneTransform!(
    self:Skinned Mesh Component Object Reference=Get GadgetComponent()
    InBoneName:Name=JNT_FX_01
    TransformSpace:ERelativeTransformSpace Enum=RTS_World
   ) -> (
    ReturnValue_Location:Vector
    ReturnValue_Rotation:Rotator
    ReturnValue_Scale_X:double
    ReturnValue_Scale_Y:double
    ReturnValue_Scale_Z:double
   )
   > SetVectorParameterValueOnMaterials!(execute, self:Mesh Component Object Reference=@n32, ParameterName:Name=Ring01Position, ParameterValue:Vector=@n33) -> (then)
   > SetScalarParameterValueOnMaterials!(execute, self:Mesh Component Object Reference=@n32, ParameterName:Name=Ring01Scale, ParameterValue:float=@n33.ReturnValue_Scale_X) -> (then)
   > n34 = GetOrCreateDynamicOverlayMaterial(MeshComponent=@n32, MaterialName=None)
   > SetVectorParameterValue(self=@n34, ParameterName=Ring01Position, Value="To LinearColor (Vector)"!(InVec:Vector=@n33) -> (ReturnValue:Linear Color Structure))
   > SetScalarParameterValue(self=@n34, ParameterName=Ring01Scale, Value=@n33.ReturnValue_Scale_X)
   n35 = GetBoneTransform(self=Get GadgetComponent(), InBoneName=JNT_FX_02, TransformSpace=RTS_World)
   > SetVectorParameterValueOnMaterials(self=@n32, ParameterName=Ring02Position, ParameterValue=@n35)
   > SetScalarParameterValueOnMaterials(self=@n32, ParameterName=Ring02Scale, ParameterValue=@n35.ReturnValue_Scale_X)
   > n36 = GetOrCreateDynamicOverlayMaterial(MeshComponent=@n32, MaterialName=None)
   > SetVectorParameterValue(self=@n36, ParameterName=Ring02Position, Value="To LinearColor (Vector)"(InVec=@n35))
   > SetScalarParameterValue(self=@n36, ParameterName=Ring02Scale, Value=@n35.ReturnValue_Scale_X)
   n37 = GetPartSkeletalMeshComponent()
   Set Disable Retargeting(
    execute=@n18.then_2
    Disable Retargeting:bool="Not Equal (Name)"(A=GetCurrentSkinWeightProfileName!(self:Skinned Mesh Component Object Reference=@n37, InLayer:ESkinWeightProfileLayer Enum=Secondary) -> (ReturnValue:Name=None)) -> (ReturnValue=~)
   ) -> (Output_Get:bool)
   n38 = Set Previous Mat(execute=@n16.else, Previous Mat=@n15)
   n39 = Event BlueprintUpdateAnimation()
   n40 = Event OnLeftArmHidden() {is=title K2Node_Event}
   Set isWearingGauntlets(execute=[@n40.then, @n43.then], isWearingGauntlets:bool=true) -> (Output_Get:bool)
   n41 = Event OnLeftArmVisible() {is=title K2Node_Event}
   Set isWearingGauntlets(execute=[@n41.then, @n42.then], isWearingGauntlets=false)
   n42 = Event OnRightArmVisible() {is=title K2Node_Event}
   n43 = Event OnRightArmHidden() {is=title K2Node_Event}
   GetOwningActor()
   GetAnimInstance(self=GetPartSkeletalMeshComponent())
   AND Boolean(A="float > float"(A=GetCurveValue(CurveName=IsGhost)), B=@n6)
   AND Boolean(A="float > float"(B=GetCurveValue(CurveName=IsGhost)), B=@n12)
  ]}
]
variables [
 Active {type=boolean}
 Array of Mats {type="array<MaterialInstanceDynamic>", default []}
 Body Overlay {type=MaterialInstanceDynamic}
 Disable Retargeting {type=boolean}
 GadgetComponent {type=SkeletalMeshComponent}
 isWearingGauntlets {type=boolean}
 Outline Material {type=MaterialInstanceDynamic}
 Previous Mat {type=MaterialInterface, default=null}
 Transformation Complete {type=boolean}
]
