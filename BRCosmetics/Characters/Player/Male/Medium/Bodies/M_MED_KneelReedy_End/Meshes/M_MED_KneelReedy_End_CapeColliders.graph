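# Physics asset of cape colliders for the M_MED_KneelReedy_End male body cosmetic, apparently the collision volumes that keep its cape out of the body.
# Capsules cover the pelvis, spine chain, both clavicles and neck_02, with a single sphere on the head.
# Arm, leg and foot bodies are listed but carry no shapes, so only the torso, shoulders, neck and head collide.
# It defines no constraints, so it serves as collision shapes rather than a jointed ragdoll.
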PhysicsAsset /BRCosmetics/Characters/Player/Male/Medium/Bodies/M_MED_KneelReedy_End/Meshes/M_MED_KneelReedy_End_CapeColliders
bodies {
 calf_l []
 calf_r []
 clavicle_l [{shapeType=Capsule, length=9.0192, radius=9.8871, center=<9.1727 2.1921 -2.4341>, rotation=<rot 69.3297 1.4476 176.9462>}]
 clavicle_r [{shapeType=Capsule, length=9.0192, radius=9.8871, center=<-9.1727 -2.1921 2.4341>, rotation=<rot 69.3297 1.4476 -3.0538>}]
 foot_l []
 foot_r []
 head [{shapeType=Sphere, radius=17.3125, center=<11.028 0.751 0>}]
 lowerarm_l []
 lowerarm_r []
 neck_02 [{shapeType=Capsule, length=5.1461, radius=5.0275, center=<-3.2996 0.3747 0>, rotation=<rot 90 4.2785 0>}]
 pelvis [{shapeType=Capsule, length=9.1802, radius=9.0465}]
 spine_01 [{shapeType=Capsule, length=6.6128, radius=9.0465}]
 spine_02 [{shapeType=Capsule, length=6.4338, radius=9.0465, center=<0.1269 0.572 0>}]
 spine_03 [{shapeType=Capsule, length=8.0738, radius=9.0465, center=<0 1.0227 0>}]
 spine_04 [{shapeType=Capsule, length=8.0504, radius=11.4858, center=<2.766 1.4525 0.0002>}]
 spine_05 [{shapeType=Capsule, length=11.2955, radius=12.0203, center=<-0.1192 2.4664 0>}]
 thigh_l []
 thigh_r []
 upperarm_l []
 upperarm_r []
}
constraints []
