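# Shared Control Rig for the KneelReedy_End male body that defines an arm retargeting and deformation function, left unwired from Begin Execution here.
# After a shared deform-rig pass, soft two-bone IK fits RetArm bones to IK hands, switching pole-vector behavior for gauntlets.
# Twist is spread along arm chains, then pose readers and rotation deltas drive shoulder, elbow and wrist correctives.
# Disabling retargeting skips these steps but keeps lat correctives.
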ControlRigBlueprint /BRCosmetics/Characters/Player/Male/Medium/Bodies/M_MED_KneelReedy_End/Meshes/M_MED_KneelReedy_End_Shared_CtrlRig
graphs [
 CollapseNode_1_ContainedGraph {nodes [
   n1 = RigVMFunctionEntryNode()
   > RigVMFunction_Sequence = RigVMAggregateNode()
   TwoBoneIKWithSoftness v3 = RigVMFunctionReferenceNode(
    bendAxis=(X=0,Y=0,Z=1)
    effectorBone=(Type=Bone,Name="handRet_l")
    firstBone=(Type=Bone,Name="upperRetArm_l")
    firstBoneOrig=(Type=Bone,Name="upperarm_l")
    ikEffector=(Type=Bone,Name="hand_ik_l")
    itemStartStretch=(Type=Null,Name="upperRetArm_l_null")
    offsetPv=(X=0,Y=0,Z=0)
    poleVector=(Type=Control,Name="arm_l_pv_ctrl")
    primaryAxis=(X=1,Y=0,Z=0)
    secondBone=(Type=Bone,Name="lowerRetArm_l")
    secondBoneConstraint=True
    secondBoneOrig=(Type=Bone,Name="lowerarm_l")
    secondaryAxis=(X=0,Y=-1,Z=0)
    softness=2
    ExecuteContext=@RigVMFunction_Sequence.A
    lockPV=@n1.lockPV
   )
   > TwoBoneIKWithSoftness v3_1 = RigVMFunctionReferenceNode(
    bendAxis=(X=0,Y=0,Z=1)
    effectorBone=(Type=Bone,Name="handRet_r")
    firstBone=(Type=Bone,Name="upperRetArm_r")
    firstBoneOrig=(Type=Bone,Name="upperarm_r")
    ikEffector=(Type=Bone,Name="hand_ik_r")
    itemStartStretch=(Type=Null,Name="upperRetArm_r_null")
    offsetPv=(X=0,Y=0,Z=0)
    poleVector=(Type=Control,Name="arm_r_pv_ctrl")
    primaryAxis=(X=-1,Y=0,Z=0)
    secondBone=(Type=Bone,Name="lowerRetArm_r")
    secondBoneConstraint=True
    secondBoneOrig=(Type=Bone,Name="lowerarm_r")
    secondaryAxis=(X=0,Y=1,Z=0)
    softness=2
    lockPV=@n1.lockPV
   )
   DeltaQuatExtraction = RigVMFunctionReferenceNode(
    child=(Type=Bone,Name="upperRetArm_l")
    invert=true
    onlyTwist=true
    parent=(Type=Null,Name="upperRetArm_l_null")
    twistAxis=(X=1,Y=0,Z=0)
    ExecuteContext=@RigVMFunction_Sequence.B
   )
   ItemArray = RigVMDispatchNode(Value=((Type=Bone,Name="upperRetArm_twist_01_l"),(Type=Bone,Name="upperRetArm_twist_02_l"),(Type=Bone,Name="upperRetArm_twist_03_l"),(Type=None,Name="None"))) {template="DISPATCH_RigVMDispatch_Constant(io Value)"}
   > SplineTwistTopBot = RigVMFunctionReferenceNode(botRotQuat=(), enable=true, topRotQuat=@DeltaQuatExtraction.outputQuat, chainArray=@ItemArray.Value)
   > DeltaQuatExtraction_1 = RigVMFunctionReferenceNode(child=(Type=Bone,Name="hand_l"), invert=false, onlyTwist=true, parent=(Type=Null,Name="hand_l_null"), twistAxis=(X=1,Y=0,Z=0))
   ItemArray_2 = RigVMDispatchNode(Value=((Type=None,Name="None"),(Type=Bone,Name="lowerRetArm_twist_01_l"),(Type=Bone,Name="lowerRetArm_twist_02_l"),(Type=Bone,Name="lowerRetArm_twist_03_l"),(Type=Bone,Name="lowerRetArm_twist_04_l"))) {template="DISPATCH_RigVMDispatch_Constant(io Value)"}
   > SplineTwistTopBot_1 = RigVMFunctionReferenceNode(enable=true, topRotQuat=(), botRotQuat=@DeltaQuatExtraction_1.outputQuat, chainArray=@ItemArray_2.Value)
   > DeltaQuatExtraction_2 = RigVMFunctionReferenceNode(child=(Type=Bone,Name="upperRetArm_r"), invert=true, onlyTwist=true, parent=(Type=Null,Name="upperRetArm_r_null"), twistAxis=(X=1,Y=0,Z=0))
   ItemArray_1 = RigVMDispatchNode(Value=((Type=Bone,Name="upperRetArm_twist_01_r"),(Type=Bone,Name="upperRetArm_twist_02_r"),(Type=Bone,Name="upperRetArm_twist_03_r"),(Type=None,Name="None"))) {template="DISPATCH_RigVMDispatch_Constant(io Value)"}
   > SplineTwistTopBot_3 = RigVMFunctionReferenceNode(botRotQuat=(), enable=true, topRotQuat=@DeltaQuatExtraction_2.outputQuat, chainArray=@ItemArray_1.Value)
   > DeltaQuatExtraction_3 = RigVMFunctionReferenceNode(child=(Type=Bone,Name="hand_r"), invert=false, onlyTwist=true, parent=(Type=Null,Name="hand_r_null"), twistAxis=(X=1,Y=0,Z=0))
   ItemArray_3 = RigVMDispatchNode(Value=((Type=None,Name="None"),(Type=Bone,Name="lowerRetArm_twist_01_r"),(Type=Bone,Name="lowerRetArm_twist_02_r"),(Type=Bone,Name="lowerRetArm_twist_03_r"),(Type=Bone,Name="lowerRetArm_twist_04_r"))) {template="DISPATCH_RigVMDispatch_Constant(io Value)"}
   > SplineTwistTopBot_2 = RigVMFunctionReferenceNode(enable=true, topRotQuat=(), botRotQuat=@DeltaQuatExtraction_3.outputQuat, chainArray=@ItemArray_3.Value)
   RigVMFunctionReturnNode(ExecuteContext=@RigVMFunction_Sequence.C)
  ]}
 CollapseNode_ContainedGraph {nodes [
   n1 = RigVMFunctionEntryNode()
   upperarms = RigVMCollapseNode(ExecuteContext=@n1.ExecutePin)
   elbows = RigVMCollapseNode(ExecutePin=@upperarms.ExecuteContext)
   wrists = RigVMCollapseNode(ExecutePin=@elbows.ExecutePin)
   RigVMFunctionReturnNode(ExecutePin=@wrists.ExecutePin)
  ]}
 CollapseNode_ContainedGraph {nodes [
   RigVMFunctionEntryNode()
   > RigVMFunction_Sequence = RigVMAggregateNode()
   &RigUnit_ParentConstraint(
    AdvancedSettings=(RotationOrderForFilter=XZY)
    Child=(Type=Bone,Name="deform_upperarm_up_base_l")
    Parents=((Item=(Type=Bone,Name="upperRetArm_twist_01_l"),Weight=0.5),(Item=(Type=Bone,Name="clavicle_l"),Weight=0.5))
    Weight=1
    ExecutePin=@RigVMFunction_Sequence.A
   )
   > &RigUnit_ParentConstraint(
    AdvancedSettings=(RotationOrderForFilter=XZY)
    Child=(Type=Bone,Name="deform_upperarm_up_base_r")
    Parents=((Item=(Type=Bone,Name="upperRetArm_twist_01_r"),Weight=0.5),(Item=(Type=Bone,Name="clavicle_r"),Weight=0.5))
    Weight=1
   )
   > n1 = &RigUnit_SphericalPoseReader(
    ActiveRegionScaleFactors=(PositiveWidth=2.5,NegativeWidth=3,PositiveHeight=3,NegativeHeight=3)
    ActiveRegionSize=0.1
    Debug=(bDrawDebug=false,DebugScale=25,DebugSegments=20,DebugThickness=0.25)
    DriverAxis=(X=1,Y=0,Z=0)
    DriverItem=(Type=Bone,Name="upperRetArm_l")
    FalloffRegionScaleFactors=(PositiveWidth=1,NegativeWidth=1,PositiveHeight=1,NegativeHeight=1)
    FalloffSize=0.5
    OptionalParentItem=(Type=Bone)
    RotationOffset=(X=0,Y=0,Z=0)
   )
   > &RigUnit_ModifyTransforms(
    ItemToModify=((Item=(Type=Bone,Name="deform_upperarm_up_l"),Transform=(Rotation=(X=0,Y=-0.0872,Z=0,W=0.9962),Translation=(X=0,Y=0,Z=0),Scale3D=(X=1,Y=1,Z=0.2))))
    Weight=@n1.OutputParam
   )
   > n2 = &RigUnit_SphericalPoseReader(
    ActiveRegionScaleFactors=(PositiveWidth=2.5,NegativeWidth=3,PositiveHeight=3,NegativeHeight=3)
    ActiveRegionSize=0.1
    Debug=(bDrawDebug=false,DebugScale=25,DebugSegments=20,DebugThickness=0.25)
    DriverAxis=(X=-1,Y=0,Z=0)
    DriverItem=(Type=Bone,Name="upperRetArm_r")
    FalloffRegionScaleFactors=(PositiveWidth=1,NegativeWidth=1,PositiveHeight=1,NegativeHeight=1)
    FalloffSize=0.5
    FlipHeightScaling=true
    OptionalParentItem=(Type=Bone)
    RotationOffset=(X=180,Y=0,Z=180)
   )
   > &RigUnit_ModifyTransforms(
    ItemToModify=((Item=(Type=Bone,Name="deform_upperarm_up_r"),Transform=(Rotation=(X=0,Y=-0.0872,Z=0,W=0.9962),Translation=(X=0,Y=0,Z=0),Scale3D=(X=1,Y=1,Z=0.2))))
    Weight=@n2.OutputParam
   )
   &RigUnit_ParentConstraint(
    AdvancedSettings=(RotationOrderForFilter=XZY)
    Child=(Type=Bone,Name="deform_upperarm_fr_base_l")
    Parents=((Item=(Type=Bone,Name="upperRetArm_twist_01_l"),Weight=0.3),(Item=(Type=Bone,Name="clavicle_l"),Weight=0.7))
    Weight=1
    ExecutePin=@RigVMFunction_Sequence.B
   )
   > &RigUnit_ParentConstraint(
    AdvancedSettings=(RotationOrderForFilter=XZY)
    Child=(Type=Bone,Name="deform_upperarm_fr_base_r")
    Parents=((Item=(Type=Bone,Name="upperRetArm_twist_01_r"),Weight=0.3),(Item=(Type=Bone,Name="clavicle_r"),Weight=0.7))
    Weight=1
   )
   > n3 = &RigUnit_SphericalPoseReader(
    ActiveRegionScaleFactors=(PositiveWidth=0.5,NegativeWidth=3,PositiveHeight=1.5,NegativeHeight=1.5)
    ActiveRegionSize=0.1
    Debug=(bDrawDebug=false,DebugScale=25,DebugSegments=20,DebugThickness=0.25)
    DriverAxis=(X=1,Y=0,Z=0)
    DriverItem=(Type=Bone,Name="upperRetArm_l")
    FalloffRegionScaleFactors=(PositiveWidth=1,NegativeWidth=1,PositiveHeight=1,NegativeHeight=1)
    FalloffSize=0.4
    OptionalParentItem=(Type=Bone)
    RotationOffset=(X=90,Y=0,Z=0)
   )
   > &RigUnit_ModifyTransforms(
    ItemToModify="((Item=(Type=Bone,Name=\"deform_upperarm_fr_l\"),Transform=(Rotation=(X=0,Y=0,Z=0,W=1),Translation=(X=0,Y=6,Z=0),Scale3D=(X=0.5,Y=1,Z=1))),(Item=(Type=Bone,Name=\"deform_upperarm_bk_l\"),Transform=(Rotation=(X=0,Y=0,Z=0,W=1),Translation=(X=0,Y=-6,Z=0),Scale3D= ...(truncated)"
    Weight=@n3.OutputParam
   )
   > n4 = &RigUnit_SphericalPoseReader(
    ActiveRegionScaleFactors=(PositiveWidth=0.5,NegativeWidth=3,PositiveHeight=1.5,NegativeHeight=1.5)
    ActiveRegionSize=0.1
    Debug=(bDrawDebug=false,DebugScale=25,DebugSegments=20,DebugThickness=0.25)
    DriverAxis=(X=-1,Y=0,Z=0)
    DriverItem=(Type=Bone,Name="upperRetArm_r")
    FalloffRegionScaleFactors=(PositiveWidth=1,NegativeWidth=1,PositiveHeight=1,NegativeHeight=1)
    FalloffSize=0.4
    FlipWidthScaling=true
    OptionalParentItem=(Type=Bone)
    RotationOffset=(X=90,Y=180,Z=180)
   )
   > &RigUnit_ModifyTransforms(
    ItemToModify="((Item=(Type=Bone,Name=\"deform_upperarm_fr_r\"),Transform=(Rotation=(X=0,Y=0,Z=0,W=1),Translation=(X=0,Y=-6,Z=0),Scale3D=(X=0.5,Y=1,Z=1))),(Item=(Type=Bone,Name=\"deform_upperarm_bk_r\"),Transform=(Rotation=(X=0,Y=0,Z=0,W=1),Translation=(X=0,Y=6,Z=0),Scale3D= ...(truncated)"
    Weight=@n4.OutputParam
   )
   &RigUnit_ParentConstraint(
    AdvancedSettings=(RotationOrderForFilter=XZY)
    Child=(Type=Bone,Name="deform_upperarm_bk_base_l")
    Parents=((Item=(Type=Bone,Name="upperRetArm_twist_01_l"),Weight=0.3),(Item=(Type=Bone,Name="clavicle_l"),Weight=0.7))
    Weight=1
    ExecutePin=@RigVMFunction_Sequence.C
   )
   > &RigUnit_ParentConstraint(
    AdvancedSettings=(RotationOrderForFilter=XZY)
    Child=(Type=Bone,Name="deform_upperarm_bk_base_r")
    Parents=((Item=(Type=Bone,Name="upperRetArm_twist_01_r"),Weight=0.3),(Item=(Type=Bone,Name="clavicle_r"),Weight=0.7))
    Weight=1
   )
   > n5 = &RigUnit_SphericalPoseReader(
    ActiveRegionScaleFactors=(PositiveWidth=3,NegativeWidth=1,PositiveHeight=1.5,NegativeHeight=1.5)
    ActiveRegionSize=0.1
    Debug=(bDrawDebug=false,DebugScale=25,DebugSegments=20,DebugThickness=0.25)
    DriverAxis=(X=1,Y=0,Z=0)
    DriverItem=(Type=Bone,Name="upperRetArm_l")
    FalloffRegionScaleFactors=(PositiveWidth=1,NegativeWidth=1,PositiveHeight=1,NegativeHeight=1)
    FalloffSize=0.4
    OptionalParentItem=(Type=Bone)
    RotationOffset=(X=90,Y=0,Z=180)
   )
   > &RigUnit_ModifyTransforms(
    ItemToModify=((Item=(Type=Bone,Name="deform_upperarm_bk_l"),Transform=(Rotation=(X=0,Y=0,Z=0,W=1),Translation=(X=0,Y=-4,Z=0),Scale3D=(X=1,Y=1,Z=1))))
    Weight=@n5.OutputParam
   )
   > n6 = &RigUnit_SphericalPoseReader(
    ActiveRegionScaleFactors=(PositiveWidth=3,NegativeWidth=1,PositiveHeight=1.5,NegativeHeight=1.5)
    ActiveRegionSize=0.1
    Debug=(bDrawDebug=false,DebugScale=25,DebugSegments=20,DebugThickness=0.25)
    DriverAxis=(X=-1,Y=0,Z=0)
    DriverItem=(Type=Bone,Name="upperRetArm_r")
    FalloffRegionScaleFactors=(PositiveWidth=1,NegativeWidth=1,PositiveHeight=1,NegativeHeight=1)
    FalloffSize=0.4
    FlipHeightScaling=true
    FlipWidthScaling=false
    OptionalParentItem=(Type=Bone)
    RotationOffset=(X=90,Y=0,Z=0)
   )
   > &RigUnit_ModifyTransforms(
    ItemToModify=((Item=(Type=Bone,Name="deform_upperarm_bk_r"),Transform=(Rotation=(X=0,Y=0,Z=0,W=1),Translation=(X=0,Y=4,Z=0),Scale3D=(X=1,Y=1,Z=1))))
    Weight=@n6.OutputParam
   )
   RigVMFunctionReturnNode(ExecuteContext=@RigVMFunction_Sequence.D)
  ]}
 CollapseNode_ContainedGraph {nodes [
   n1 = RigVMFunctionEntryNode()
   &RigUnit_ParentConstraint(
    AdvancedSettings=(RotationOrderForFilter=XZY)
    Child=(Type=Bone,Name="deform_elbow_in_base_l")
    Parents=((Item=(Type=Bone,Name="upperRetArm_l"),Weight=0.6),(Item=(Type=Bone,Name="lowerRetArm_l"),Weight=0.4))
    Weight=1
    ExecutePin=@n1.ExecutePin
   )
   > &RigUnit_ParentConstraint(
    AdvancedSettings=(RotationOrderForFilter=XZY)
    Child=(Type=Bone,Name="deform_elbow_out_base_l")
    Parents=((Item=(Type=Bone,Name="upperRetArm_l"),Weight=0.6),(Item=(Type=Bone,Name="lowerRetArm_l"),Weight=0.4))
    Weight=1
   )
   > DeltaRotation v4 = RigVMFunctionReferenceNode(
    Rotation Order=ZYX
    child=(Type=Bone,Name="lowerRetArm_l")
    outputMinMax=(X=0,Y=1)
    parent=(Type=Null,Name="lowerRetArm_l_null")
    premult=1
    rotationMinMax=(X=15,Y=90)
    triggerAxis=Z
    twistAxis=(X=1,Y=0,Z=0)
   )
   > &RigUnit_ModifyTransforms(
    ItemToModify="((Item=(Type=Bone,Name=\"deform_elbow_in_l\"),Transform=(Rotation=(X=0,Y=0,Z=0,W=-1),Translation=(X=0,Y=9,Z=0),Scale3D=(X=0.5,Y=1,Z=1))),(Item=(Type=Bone,Name=\"deform_elbow_out_l\"),Transform=(Rotation=(X=0,Y=0,Z=0,W=1),Translation=(X=0,Y=-2,Z=0),Scale3D=(X=0 ...(truncated)"
    WeightMinimum=0
    Weight=@"DeltaRotation v4".outputWeight
   )
   > &RigUnit_ParentConstraint(
    AdvancedSettings=(RotationOrderForFilter=XZY)
    Child=(Type=Bone,Name="deform_elbow_in_base_r")
    Parents=((Item=(Type=Bone,Name="upperRetArm_r"),Weight=0.6),(Item=(Type=Bone,Name="lowerRetArm_r"),Weight=0.4))
    Weight=1
   )
   > &RigUnit_ParentConstraint(
    AdvancedSettings=(RotationOrderForFilter=XZY)
    Child=(Type=Bone,Name="deform_elbow_out_base_r")
    Parents=((Item=(Type=Bone,Name="upperRetArm_r"),Weight=0.6),(Item=(Type=Bone,Name="lowerRetArm_r"),Weight=0.4))
    Weight=1
   )
   > DeltaRotation v4_1 = RigVMFunctionReferenceNode(
    Rotation Order=ZYX
    child=(Type=Bone,Name="lowerRetArm_r")
    outputMinMax=(X=0,Y=1)
    parent=(Type=Null,Name="lowerRetArm_r_null")
    premult=1
    rotationMinMax=(X=15,Y=90)
    triggerAxis=Z
    twistAxis=(X=1,Y=0,Z=0)
   )
   > n2 = &RigUnit_ModifyTransforms(
    ItemToModify="((Item=(Type=Bone,Name=\"deform_elbow_in_r\"),Transform=(Rotation=(X=0,Y=0,Z=0,W=-1),Translation=(X=0,Y=-9,Z=0),Scale3D=(X=0.5,Y=1,Z=1))),(Item=(Type=Bone,Name=\"deform_elbow_out_r\"),Transform=(Rotation=(X=0,Y=0,Z=0,W=1),Translation=(X=0,Y=2,Z=0),Scale3D=(X=0 ...(truncated)"
    WeightMinimum=0
    Weight=@"DeltaRotation v4_1".outputWeight
   )
   RigVMFunctionReturnNode(ExecutePin=@n2.ExecutePin)
  ]}
 CollapseNode_ContainedGraph {nodes [
   n1 = RigVMFunctionEntryNode()
   &RigUnit_ParentConstraint(
    AdvancedSettings=(RotationOrderForFilter=XZY)
    Child=(Type=Bone,Name="deform_wrist_up_base_l")
    Parents=((Item=(Type=Bone,Name="hand_l"),Weight=0.4),(Item=(Type=Bone,Name="lowerRetArm_twist_04_l"),Weight=0.6))
    Weight=1
    ExecutePin=@n1.ExecutePin
   )
   > &RigUnit_ParentConstraint(
    AdvancedSettings=(RotationOrderForFilter=XZY)
    Child=(Type=Bone,Name="deform_wrist_dn_base_l")
    Parents=((Item=(Type=Bone,Name="hand_l"),Weight=0.4),(Item=(Type=Bone,Name="lowerRetArm_twist_04_l"),Weight=0.6))
    Weight=1
   )
   > n2 = &RigUnit_SphericalPoseReader(
    ActiveRegionScaleFactors=(PositiveWidth=0.5,NegativeWidth=3.5,PositiveHeight=3,NegativeHeight=3)
    ActiveRegionSize=0.1
    Debug=(bDrawDebug=false,DebugScale=25,DebugSegments=20,DebugThickness=0.25)
    DriverAxis=(X=1,Y=0,Z=0)
    DriverItem=(Type=Bone,Name="hand_l")
    FalloffRegionScaleFactors=(PositiveWidth=1,NegativeWidth=1,PositiveHeight=1,NegativeHeight=1)
    FalloffSize=0.45
    OptionalParentItem=(Type=Bone,Name="lowerRetArm_twist_04_l")
    RotationOffset=(X=90,Y=0,Z=0)
   )
   > &RigUnit_ModifyTransforms(
    ItemToModify=((Item=(Type=Bone,Name="deform_wrist_up_l"),Transform=(Rotation=(X=0,Y=0,Z=0,W=1),Translation=(X=0,Y=5,Z=0),Scale3D=(X=0.7,Y=1,Z=1))))
    Weight=@n2.OutputParam
   )
   > n3 = &RigUnit_SphericalPoseReader(
    ActiveRegionScaleFactors=(PositiveWidth=1,NegativeWidth=4,PositiveHeight=3,NegativeHeight=3)
    ActiveRegionSize=0.1
    Debug=(bDrawDebug=false,DebugScale=25,DebugSegments=20,DebugThickness=0.25)
    DriverAxis=(X=1,Y=0,Z=0)
    DriverItem=(Type=Bone,Name="hand_l")
    FalloffRegionScaleFactors=(PositiveWidth=1,NegativeWidth=1,PositiveHeight=1,NegativeHeight=1)
    FalloffSize=0.4
    OptionalParentItem=(Type=Bone,Name="lowerRetArm_twist_04_l")
    RotationOffset=(X=-90,Y=0,Z=0)
   )
   > &RigUnit_ModifyTransforms(
    ItemToModify="((Item=(Type=Bone,Name=\"deform_wrist_dn_l\"),Transform=(Rotation=(X=0,Y=0,Z=0,W=1),Translation=(X=0,Y=-6,Z=0),Scale3D=(X=0.7,Y=1,Z=1))),(Item=(Type=Bone,Name=\"deform_wrist_up_l\"),Transform=(Rotation=(X=0,Y=0,Z=0,W=1),Translation=(X=0,Y=3,Z=0),Scale3D=(X=0.2 ...(truncated)"
    Weight=@n3.OutputParam
   )
   > &RigUnit_ParentConstraint(
    AdvancedSettings=(RotationOrderForFilter=XZY)
    Child=(Type=Bone,Name="deform_wrist_up_base_r")
    Parents=((Item=(Type=Bone,Name="hand_r"),Weight=0.4),(Item=(Type=Bone,Name="lowerRetArm_twist_04_r"),Weight=0.6))
    Weight=1
   )
   > &RigUnit_ParentConstraint(
    AdvancedSettings=(RotationOrderForFilter=XZY)
    Child=(Type=Bone,Name="deform_wrist_dn_base_r")
    Parents=((Item=(Type=Bone,Name="hand_r"),Weight=0.4),(Item=(Type=Bone,Name="lowerRetArm_twist_04_r"),Weight=0.6))
    Weight=1
   )
   > n4 = &RigUnit_SphericalPoseReader(
    ActiveRegionScaleFactors=(PositiveWidth=0.5,NegativeWidth=3.5,PositiveHeight=3,NegativeHeight=3)
    ActiveRegionSize=0.1
    Debug=(bDrawDebug=false,DebugScale=25,DebugSegments=20,DebugThickness=0.25)
    DriverAxis=(X=-1,Y=0,Z=0)
    DriverItem=(Type=Bone,Name="hand_r")
    FalloffRegionScaleFactors=(PositiveWidth=1,NegativeWidth=1,PositiveHeight=1,NegativeHeight=1)
    FalloffSize=0.45
    FlipHeightScaling=true
    OptionalParentItem=(Type=Bone,Name="lowerRetArm_twist_04_r")
    RotationOffset=(X=90,Y=0,Z=180)
   )
   > &RigUnit_ModifyTransforms(
    ItemToModify=((Item=(Type=Bone,Name="deform_wrist_up_r"),Transform=(Rotation=(X=0,Y=0,Z=0,W=1),Translation=(X=0,Y=-5,Z=0),Scale3D=(X=0.7,Y=1,Z=1))))
    Weight=@n4.OutputParam
   )
   > n5 = &RigUnit_SphericalPoseReader(
    ActiveRegionScaleFactors=(PositiveWidth=1,NegativeWidth=4,PositiveHeight=3,NegativeHeight=3)
    ActiveRegionSize=0.1
    Debug=(bDrawDebug=false,DebugScale=25,DebugSegments=20,DebugThickness=0.25)
    DriverAxis=(X=-1,Y=0,Z=0)
    DriverItem=(Type=Bone,Name="hand_r")
    FalloffRegionScaleFactors=(PositiveWidth=1,NegativeWidth=1,PositiveHeight=1,NegativeHeight=1)
    FalloffSize=0.4
    FlipHeightScaling=true
    OptionalParentItem=(Type=Bone,Name="lowerRetArm_twist_04_r")
    RotationOffset=(X=-90,Y=0,Z=180)
   )
   > n6 = &RigUnit_ModifyTransforms(
    ItemToModify="((Item=(Type=Bone,Name=\"deform_wrist_dn_r\"),Transform=(Rotation=(X=0,Y=0,Z=0,W=1),Translation=(X=0,Y=6,Z=0),Scale3D=(X=0.7,Y=1,Z=1))),(Item=(Type=Bone,Name=\"deform_wrist_up_r\"),Transform=(Rotation=(X=0,Y=0,Z=0,W=1),Translation=(X=0,Y=-3,Z=0),Scale3D=(X=0.2 ...(truncated)"
    Weight=@n5.OutputParam
   )
   RigVMFunctionReturnNode(ExecutePin=@n6.ExecutePin)
  ]}
 CollapseNode_ContainedGraph {nodes [
   RigVMFunctionEntryNode()
   > DeltaRotation v4 = RigVMFunctionReferenceNode(
    Rotation Order=ZYX
    child=(Type=Bone,Name="hand_l")
    onlyTwist=True
    outputMinMax=(X=0,Y=1)
    parent=(Type=Null,Name="hand_l_null")
    premult=1
    rotationMinMax=(X=100,Y=140)
    twistAxis=(X=1,Y=0,Z=0)
   )
   > &RigUnit_ModifyTransforms(
    ItemToModify="((Item=(Type=Bone,Name=\"lowerRetArm_twist_02_l\"),Transform=(Rotation=(X=0,Y=0,Z=0,W=-1),Translation=(X=0,Y=0,Z=0),Scale3D=(X=1,Y=1.3664,Z=1.6129))),(Item=(Type=Bone,Name=\"lowerRetArm_twist_03_l\"),Transform=(Rotation=(X=0,Y=0,Z=0,W=-1),Translation=(X=0,Y=0, ...(truncated)"
    WeightMinimum=0
    Weight=@"DeltaRotation v4".outputWeight
   )
   > RigVMFunctionReturnNode()
  ]}
 CollapseNode_ContainedGraph {nodes [
   n1 = RigVMFunctionEntryNode()
   &RigUnit_ParentConstraint(
    AdvancedSettings=(RotationOrderForFilter=XZY)
    Child=(Type=Bone,Name="deform_lat_l_base")
    Filter=(TranslationFilter=(bX=True,bY=True,bZ=False),RotationFilter=(bX=False,bY=False,bZ=False),ScaleFilter=(bX=False,bY=False,bZ=False))
    Parents="((Item=(Type=Bone,Name=\"spine_02\"),Weight=0.0148),(Item=(Type=Bone,Name=\"deform_pec_l\"),Weight=0.0458),(Item=(Type=Bone,Name=\"spine_03\"),Weight=0.0611),(Item=(Type=Bone,Name=\"spine_05\"),Weight=0.086),(Item=(Type=Bone,Name=\"spine_04\"),Weight=0.3063),(Item=( ...(truncated)"
    Weight=1
    ExecutePin=@n1.ExecutePin
   )
   > &RigUnit_ParentConstraint(
    AdvancedSettings=(RotationOrderForFilter=XZY)
    Child=(Type=Bone,Name="deform_lat_r_base")
    Filter=(TranslationFilter=(bX=True,bY=True,bZ=False),RotationFilter=(bX=False,bY=False,bZ=False),ScaleFilter=(bX=False,bY=False,bZ=False))
    Parents="((Item=(Type=Bone,Name=\"spine_02\"),Weight=0.0148),(Item=(Type=Bone,Name=\"deform_pec_r\"),Weight=0.0458),(Item=(Type=Bone,Name=\"spine_03\"),Weight=0.0611),(Item=(Type=Bone,Name=\"spine_05\"),Weight=0.086),(Item=(Type=Bone,Name=\"spine_04\"),Weight=0.3063),(Item=( ...(truncated)"
    Weight=1
   )
   > DeltaRotation v4 = RigVMFunctionReferenceNode(
    Rotation Order=ZYX
    child=(Type=Bone,Name="clavicle_l")
    outputMinMax=(X=0,Y=1)
    parent=(Type=Null,Name="clavicle_l_null")
    premult=1
    rotationMinMax=(X=0,Y=-40)
    triggerAxis=Y
    twistAxis=(X=1,Y=0,Z=0)
   )
   > &RigUnit_ModifyTransforms(
    ItemToModify=((Item=(Type=Bone,Name="deform_lat_l"),Transform=(Rotation=(X=0,Y=-0.1564,Z=0,W=0.9877),Translation=(X=0,Y=0,Z=6),Scale3D=(X=1,Y=1,Z=1))))
    WeightMinimum=0
    Weight=@"DeltaRotation v4".outputWeight
   )
   > DeltaRotation v4_1 = RigVMFunctionReferenceNode(
    Rotation Order=ZYX
    child=(Type=Bone,Name="clavicle_r")
    outputMinMax=(X=0,Y=1)
    parent=(Type=Null,Name="clavicle_r_null")
    premult=1
    rotationMinMax=(X=0,Y=-40)
    triggerAxis=Y
    twistAxis=(X=1,Y=0,Z=0)
   )
   > n2 = &RigUnit_ModifyTransforms(
    ItemToModify=((Item=(Type=Bone,Name="deform_lat_r"),Transform=(Rotation=(X=0,Y=-0.1564,Z=0,W=0.9877),Translation=(X=0,Y=0,Z=6),Scale3D=(X=1,Y=1,Z=1))))
    WeightMinimum=0
    Weight=@"DeltaRotation v4_1".outputWeight
   )
   RigVMFunctionReturnNode(ExecutePin=@n2.ExecutePin)
  ]}
 New Function_ContainedGraph {nodes [
   n1 = RigVMFunctionEntryNode()
   RigVMFunction_MathBoolToInteger = &RigVMFunction_MathBoolToInteger(Value=@n1.onlyTwist)
   RigVMFunction_MathQuaternionSwingTwist = &RigVMFunction_MathQuaternionSwingTwist(
    TwistAxis=@n1.twistAxis
    Input=&RigUnit_GetRelativeTransformForItem(bChildInitial=False, bParentInitial=False, Child=@n1.child, Parent=@n1.parent).RelativeTransform.Rotation
   )
   n2 = RigVMDispatchNode(
    Values=((),())
    Index=@RigVMFunction_MathBoolToInteger.Result
    Values.0=@RigVMFunction_MathQuaternionSwingTwist.Swing
    Values.1=@RigVMFunction_MathQuaternionSwingTwist.Twist
   ) {template="DISPATCH_RigVMDispatch_SelectInt32(in Index,in Values,out Result)"}
   RigVMFunction_MathQuaternionToEuler = &RigVMFunction_MathQuaternionToEuler(Value=@n2.Result, RotationOrder=@n1.Rotation Order)
   > DISPATCH_RigVMDispatch_Print = RigVMDispatchNode(ScreenColor=(R=1,G=1,B=1,A=1), ScreenDuration=1.5, Enabled=@n1.debugRotation, Value=@RigVMFunction_MathQuaternionToEuler.Result) {template="DISPATCH_RigVMDispatch_Print(in Prefix,in Value,in Enabled,in ScreenDuration,in ScreenColor)"}
   Cast = RigVMDispatchNode(Value=@n1.triggerAxis) {template="DISPATCH_RigVMDispatch_CastEnumToInt(in Value,out Result)"}
   DISPATCH_RigVMDispatch_SelectInt32 = RigVMDispatchNode(
    Values=(0,0,0)
    Index=@Cast.Result
    Values.0=@RigVMFunction_MathQuaternionToEuler.Result.X
    Values.1=@RigVMFunction_MathQuaternionToEuler.Result.Y
    Values.2=@RigVMFunction_MathQuaternionToEuler.Result.Z
   ) {template="DISPATCH_RigVMDispatch_SelectInt32(in Index,in Values,out Result)"}
   Remap = &RigVMFunction_MathDoubleRemap(
    bClamp=False
    Value=@DISPATCH_RigVMDispatch_SelectInt32.Result
    TargetMinimum=@n1.outputMinMax.X
    TargetMaximum=@n1.outputMinMax.Y
    SourceMinimum=@n1.rotationMinMax.X
    SourceMaximum=@n1.rotationMinMax.Y
   )
   Absolute = &RigVMFunction_MathDoubleAbs(Value=@Remap.Result)
   Multiply = &RigVMFunction_MathDoubleMul(
    A=&RigVMFunction_MathDoubleClamp(Value=RigVMDispatchNode(True=@Absolute.Result, Condition=@n1.absoluteValue, False=@Remap.Result) {template="DISPATCH_RigVMDispatch_If(in Condition,in True,in False,out Result)"}.Result).Result
    B=@n1.premult
   )
   > RigVMFunctionReturnNode(outputQuat=@n2.Result, outputWeight=@Multiply.Result)
  ]}
 New Function_ContainedGraph {nodes [
   n1 = RigVMFunctionEntryNode()
   RigVMFunction_ControlFlowBranch = &RigVMFunction_ControlFlowBranch(ExecuteContext=@n1.ExecuteContext, Condition=@n1.enable)
   RigVMFunctionReturnNode(ExecuteContext=@RigVMFunction_ControlFlowBranch.Completed)
   For_Each = RigVMDispatchNode(Array=@n1.chainArray, ExecuteContext=@RigVMFunction_ControlFlowBranch.True) {template="DISPATCH_RigVMDispatch_ArrayIterator(in Array,out Element,out Index,out Count,out Ratio)"}
   Interpolate = /Script/RigVM.RigVMFunction_MathQuaternionSlerp(T=@For_Each.Ratio, B=@n1.botRotQuat, A=@n1.topRotQuat) {is=struct RigVMUnitNode, template="Interpolate::Execute(in A,in B,in T,out Result)"}
   > &RigUnit_OffsetTransformForItem(
    OffsetTransform=(Rotation=(X=0,Y=0,Z=0,W=1),Translation=(X=0,Y=0,Z=0),Scale3D=(X=1,Y=1,Z=1))
    Weight=1
    bPropagateToChildren=false
    OffsetTransform.Rotation=@Interpolate.Result
    Item=@For_Each.Element
   )
  ]}
 New Function_ContainedGraph {nodes [
   n1 = RigVMFunctionEntryNode()
   RigVMFunction_MathBoolToInteger = &RigVMFunction_MathBoolToInteger(Value=@n1.onlyTwist)
   RigVMFunction_MathQuaternionSwingTwist = &RigVMFunction_MathQuaternionSwingTwist(
    TwistAxis=@n1.twistAxis
    Input=&RigUnit_GetRelativeTransformForItem(bChildInitial=False, bParentInitial=False, Child=@n1.child, Parent=@n1.parent).RelativeTransform.Rotation
   )
   n2 = RigVMDispatchNode(
    Values=((),())
    Index=@RigVMFunction_MathBoolToInteger.Result
    Values.0=@RigVMFunction_MathQuaternionSwingTwist.Swing
    Values.1=@RigVMFunction_MathQuaternionSwingTwist.Twist
   ) {template="DISPATCH_RigVMDispatch_SelectInt32(in Index,in Values,out Result)"}
   Inverse = /Script/RigVM.RigVMFunction_MathQuaternionInverse(Value=@n2.Result) {is=struct RigVMUnitNode, template="Inverse::Execute(in Value,out Result)"}
   > RigVMFunctionReturnNode(outputQuat=RigVMDispatchNode(True=@Inverse.Result, Condition=@n1.invert, False=@n2.Result) {template="DISPATCH_RigVMDispatch_If(in Condition,in True,in False,out Result)"}.Result)
  ]}
 New Function_ContainedGraph {nodes [
   RigVMFunctionEntryNode()
   n1 = RigVMVariableNode() {variable=disable retargeting}
   > FN_DeformRig_FNC = RigVMFunctionReferenceNode(
    DeformRig_ Curve="(deformRig_l_upperarm_up_alpha_72_#1=False,deformRig_r_upperarm_up_alpha_73_#2=False,deformRig_l_upperarm_fr_alpha_74_#3=False,deformRig_r_upperarm_fr_alpha_75_CB51B8 ...(truncated)"
    DeformRig_Bool="(upperarm_1_#4=false,elbow_3_#5=false,wrist_7_#6=false,groin_8_#7=True,glute_17_#8=True,knee_9_40C408324CA ...(truncated)"
    DeformRig_Debug=(upperarm_debug_18_#4=false,wrist_debug_20_#6=False,lat_debug_26_#9=False,groinGlute_debug_27_#7=False)
    DeformRig_Value="(upperarm_up_18_#10=(Rotation=(X=0,Y=0,Z=0,W=1),Translation=(X=0,Y=0,Z=1),Scale3D=(X=1,Y=1,Z=1)),upperarm_fwd_fr_19_#11=(Rotation=(X=0,Y=0,Z=0,W=1),Translation=(X=0,Y=14,Z=0),Scale3D=(X=0.5,Y=1,Z=1) ...(truncated)"
    DeformRig_Bool.elbow_3_#5=@n1.Value
    DeformRig_Bool.wrist_7_#6=@n1.Value
   )
   RigVMFunction_ControlFlowBranch = &RigVMFunction_ControlFlowBranch(ExecuteContext=@FN_DeformRig_FNC.ExecuteContext, Condition=RigVMVariableNode() {variable=disable retargeting}.Value)
   latBones = RigVMCollapseNode(ExecutePin=@RigVMFunction_ControlFlowBranch.Completed)
   RigVMFunctionReturnNode(ExecuteContext=@latBones.ExecutePin)
   armSetup = RigVMCollapseNode(ExecuteContext=@RigVMFunction_ControlFlowBranch.False, lockPV=RigVMVariableNode() {variable=isWearingGauntlets}.Value)
   customDeforms = RigVMCollapseNode(ExecutePin=@armSetup.ExecuteContext)
   twistFix = RigVMCollapseNode(ExecuteContext=@customDeforms.ExecutePin)
  ]}
 New Function_ContainedGraph {nodes [
   RigVMFunctionEntryNode()
   n1 = &RigUnit_GetTransform(bInitial=true, Item=RigVMVariableNode() {variable=secondBone}.Value)
   n2 = &RigVMFunction_MathVectorDistance(
    B=@n1.Transform.Translation
    A=&RigUnit_GetTransform(bInitial=true, Item=RigVMVariableNode() {variable=firstBone}.Value).Transform.Translation
   )
   n3 = &RigVMFunction_MathVectorDistance(
    A=@n1.Transform.Translation
    B=&RigUnit_GetTransform(bInitial=true, Item=RigVMVariableNode() {variable=effectorBone}.Value).Transform.Translation
   )
   n4 = &RigVMFunction_MathDoubleAdd(A=@n2.Result, B=@n3.Result)
   > RigVMVariableNode(Value=@n4.Result) {variable=Entry_segmentsSum}
   > RigVMVariableNode(Value=&RigVMFunction_MathDoubleDiv(B=@n4.Result, A=@n2.Result).Result) {variable=Entry_segmentAFactor}
   > RigVMVariableNode(Value=&RigVMFunction_MathDoubleDiv(B=@n4.Result, A=@n3.Result).Result) {variable=Entry_segmentBFactor}
   n5 = RigVMVariableNode() {variable=lockPV}
   RigVMFunction_MathBoolNot = &RigVMFunction_MathBoolNot(Value=@n5.Value)
   > &RigUnit_ParentConstraint(
    AdvancedSettings=(InterpolationType=Average,RotationOrderForFilter=XZY)
    Parents=((Item=(Type=Bone,Name="None"),Weight=1))
    Weight=1
    Parents.0.Item=RigVMDispatchNode(False=RigVMVariableNode() {variable=firstBoneOrig}.Value, True=RigVMVariableNode() {variable=secondBoneOrig}.Value, Condition=@n5.Value) {template="DISPATCH_RigVMDispatch_If(in Condition,in True,in False,out Result)"}.Result
    bMaintainOffset=@RigVMFunction_MathBoolNot.Result
    Child=RigVMVariableNode() {variable=poleVector}.Value
   )
   > &RigUnit_ModifyTransforms(
    ItemToModify=((Item=(Type=Bone,Name="None"),Transform=(Rotation=(X=0,Y=0,Z=0,W=1),Translation=(X=0,Y=0,Z=0),Scale3D=(X=1,Y=1,Z=1))))
    Weight=1
    WeightMinimum=0
    ItemToModify.0.Transform.Translation=RigVMVariableNode() {variable=offsetPv}.Value
    ItemToModify.0.Item=RigVMVariableNode() {variable=poleVector}.Value
   )
   RigVMFunction_MathVectorDistance = &RigVMFunction_MathVectorDistance(
    B=&RigUnit_GetTransform(bInitial=False, Item=RigVMVariableNode() {variable=ikEffector}.Value).Transform.Translation
    A=&RigUnit_GetTransform(bInitial=False, Item=RigVMVariableNode() {variable=itemStartStretch}.Value).Transform.Translation
   )
   n6 = RigVMVariableNode() {variable=softness}
   n7 = &RigVMFunction_MathDoubleSub(B=@n6.Value, A=RigVMVariableNode() {variable=Entry_segmentsSum}.Value)
   Divide = &RigVMFunction_MathDoubleDiv(
    A=&RigVMFunction_MathDoubleMul(B=-1, A=&RigVMFunction_MathDoubleSub(A=@RigVMFunction_MathVectorDistance.Result, B=@n7.Result).Result).Result
    B=@n6.Value
   )
   Exponential = /Script/RigVM.RigVMFunction_MathDoubleExponential(Value=@Divide.Result) {is=struct RigVMUnitNode, template="Exponential::Execute(in Value,out Result)"}
   Add = &RigVMFunction_MathDoubleAdd(
    B=RigVMDispatchNode(
     False=0
     Condition=/Script/RigVM.RigVMFunction_MathDoubleGreater(A=@RigVMFunction_MathVectorDistance.Result, B=@n7.Result) {is=struct RigVMUnitNode, template="Greater::Execute(in A,in B,out Result)"}.Result
     True=&RigVMFunction_MathDoubleMul(A=@Exponential.Result, B=@n6.Value).Result
    ) {template="DISPATCH_RigVMDispatch_If(in Condition,in True,in False,out Result)"}.Result
    A=@RigVMFunction_MathVectorDistance.Result
   )
   Greater = &RigVMFunction_MathDoubleGreater(A=@Add.Result, B=RigVMVariableNode() {variable=Entry_segmentsSum}.Value) {template="Greater::Execute(in A,in B,out Result)"}
   n8 = RigVMDispatchNode(True=@Add.Result, Condition=@Greater.Result, False=RigVMVariableNode() {variable=Entry_segmentsSum}.Value) {template="DISPATCH_RigVMDispatch_If(in Condition,in True,in False,out Result)"}
   Multiply = &RigVMFunction_MathDoubleMul(B=@n8.Result, A=RigVMVariableNode() {variable=Entry_segmentAFactor}.Value)
   n9 = &RigUnit_GetTransform(bInitial=false, Item=RigVMVariableNode() {variable=poleVector}.Value)
   n10 = &RigUnit_GetTransform(bInitial=false, Item=RigVMVariableNode() {variable=itemStartStretch}.Value)
   n11 = RigVMVariableNode() {variable=Entry_segmentBFactor}
   n12 = &RigUnit_GetTransform(bInitial=false, Item=RigVMVariableNode() {variable=ikEffector}.Value)
   > /Script/ControlRig.RigUnit_TwoBoneIKSimplePerItem(
    DebugSettings=(bEnabled=False,Scale=10,WorldOffset=(Rotation=(X=0,Y=0,Z=0,W=1),Translation=(X=0,Y=0,Z=0),Scale3D=(X=1,Y=1,Z=1)))
    PoleVectorKind=Direction
    PoleVectorSpace=(Type=Bone,Name="None")
    SecondaryAxisWeight=1
    StretchMaximumRatio=2
    StretchStartRatio=1
    Weight=1
    bEnableStretch=false
    bPropagateToChildren=True
    PoleVector=&RigUnit_GetTransform(bInitial=False, Item=RigVMVariableNode() {variable=poleVector}.Value).Transform.Translation
    PrimaryAxis=RigVMVariableNode() {variable=primaryAxis}.Value
    SecondaryAxis=RigVMVariableNode() {variable=secondaryAxis}.Value
    ItemA=RigVMVariableNode() {variable=firstBone}.Value
    ItemB=RigVMVariableNode() {variable=secondBone}.Value
    EffectorItem=RigVMVariableNode() {variable=effectorBone}.Value
    Effector=&RigUnit_GetTransform(bInitial=False, Item=RigVMVariableNode() {variable=ikEffector}.Value).Transform
    ItemALength=RigVMDispatchNode(
     False=@Multiply.Result
     True=&RigVMFunction_MathVectorDistance(B=@n9.Transform.Translation, A=@n10.Transform.Translation).Result
     Condition=RigVMVariableNode() {variable=lockPV}.Value
    ) {template="DISPATCH_RigVMDispatch_If(in Condition,in True,in False,out Result)"}.Result
    ItemBLength=RigVMDispatchNode(
     False=&RigVMFunction_MathDoubleMul(A=@n8.Result, B=@n11.Value).Result
     True=&RigVMFunction_MathVectorDistance(A=@n9.Transform.Translation, B=@n12.Transform.Translation).Result
     Condition=RigVMVariableNode() {variable=lockPV}.Value
    ) {template="DISPATCH_RigVMDispatch_If(in Condition,in True,in False,out Result)"}.Result
   ) {is=struct RigVMUnitNode}
   > n13 = &RigUnit_ParentConstraint(
    AdvancedSettings=(InterpolationType=Average,RotationOrderForFilter=XZY)
    Parents=((Item=(Type=Bone,Name="None"),Weight=1))
    Weight=1
    Child=RigVMVariableNode() {variable=effectorBone}.Value
    Parents.0.Item=RigVMVariableNode() {variable=secondBone}.Value
   )
   RigVMFunction_ControlFlowBranch = &RigVMFunction_ControlFlowBranch(ExecuteContext=@n13.ExecutePin, Condition=RigVMVariableNode() {variable=secondBoneConstraint}.Value)
   RigVMFunctionReturnNode(ExecuteContext=@RigVMFunction_ControlFlowBranch.Completed)
   n14 = RigVMVariableNode() {variable=bendAxis}
   &RigUnit_AimConstraintLocalSpaceOffset(
    Filter=(bX=True,bY=True,bZ=True)
    Parents=((Item=(Type=Bone,Name="None"),Weight=1))
    Weight=1
    WorldUp=(Target=(X=0,Y=0,Z=1),Kind=Direction,Space=(Type=None,Name="None"))
    bMaintainOffset=false
    UpAxis=@n14.Value
    WorldUp.Target=@n14.Value
    Child=RigVMVariableNode() {variable=secondBone}.Value
    AimAxis=RigVMVariableNode() {variable=primaryAxis}.Value
    WorldUp.Space=RigVMVariableNode() {variable=firstBone}.Value
    Parents.0.Item=RigVMVariableNode() {variable=ikEffector}.Value
    ExecutePin=@RigVMFunction_ControlFlowBranch.True
   )
  ]}
 RigVMFunctionLibrary {nodes [
   DeltaQuatExtraction = RigVMCollapseNode(ExecuteContext=(), child=(Type=Bone,Name="None"), parent=(Type=Bone,Name="None"), twistAxis=(X=1,Y=0,Z=0))
   DeltaRotation v4 = RigVMCollapseNode(
    ExecuteContext=()
    Rotation Order=ZYX
    child=(Type=Bone,Name="None")
    outputMinMax=(X=0,Y=1)
    parent=(Type=Bone,Name="None")
    premult=1
    rotationMinMax=(X=0,Y=0)
    twistAxis=(X=1,Y=0,Z=0)
   )
   KneelReedyEnd_armsRetarg = RigVMCollapseNode(ExecuteContext=(), disable retargeting=False, isWearingGauntlets=False)
   SplineTwistTopBot = RigVMCollapseNode(
    ExecuteContext=()
    botRotQuat=()
    chainArray="((Type=Bone,Name=\"arm_bend_1_l\"),(Type=Bone,Name=\"arm_bend_2_l\"),(Type=Bone,Name=\"arm_bend_3_l\"),(Type=Bone,Name=\"arm_bend_4_l\"),(Type=Bone,Name=\"arm_bend_5_l\"),(Type=Bone,Name=\"arm_bend_6_l\"),(Type=Bone,Name=\"arm_bend_7_l\"),(Type=Bone,Name=\"arm_bend_8_l\") ...(truncated)"
    enable=true
    topRotQuat=()
   )
   TwoBoneIKWithSoftness v3 = RigVMCollapseNode(
    ExecuteContext=()
    bendAxis=(X=0,Y=0,Z=1)
    effectorBone=(Type=None,Name="None")
    firstBone=(Type=None,Name="None")
    firstBoneOrig=(Type=None,Name="None")
    ikEffector=(Type=None,Name="None")
    itemStartStretch=(Type=None,Name="None")
    offsetPv=(X=0,Y=0,Z=0)
    poleVector=(Type=None,Name="None")
    primaryAxis=(X=1,Y=0,Z=0)
    secondBone=(Type=None,Name="None")
    secondBoneOrig=(Type=Bone,Name="None")
    secondaryAxis=(X=0,Y=1,Z=0)
    softness=1
   )
  ]}
 RigVMFunction_Sequence_1_ContainedGraph {nodes [
   n1 = RigVMFunctionEntryNode()
   RigVMFunction_Sequence = &RigVMFunction_Sequence(ExecuteContext=@n1.ExecuteContext)
   n2 = &RigVMFunction_Sequence(ExecuteContext=@RigVMFunction_Sequence.B)
   n3 = &RigVMFunction_Sequence(ExecuteContext=@n2.B)
   RigVMFunctionReturnNode(A=@RigVMFunction_Sequence.A, B=@n2.A, C=@n3.A, D=@n3.B)
  ]}
 RigVMFunction_Sequence_1_ContainedGraph {nodes [
   n1 = RigVMFunctionEntryNode()
   RigVMFunction_Sequence = &RigVMFunction_Sequence(ExecuteContext=@n1.ExecuteContext)
   n2 = &RigVMFunction_Sequence(ExecuteContext=@RigVMFunction_Sequence.B)
   RigVMFunctionReturnNode(A=@RigVMFunction_Sequence.A, B=@n2.A, C=@n2.B)
  ]}
 RigVMModel {nodes [
   KneelReedyEnd_armsRetarg = RigVMFunctionReferenceNode(ExecuteContext=(), disable retargeting=False, isWearingGauntlets=False)
   RigUnit_BeginExecution = &RigUnit_BeginExecution()
  ]}
]
hierarchy [
 root {loc=<0 0 0>, rot=<0 0 0 1>}
 pelvis {parent=root, loc=<0 -0.7071 93.1191>, rot=<0.0127 0.707 -0.0127 -0.707>}
 spine_01 {parent=pelvis, loc=<6.7494 -0.6388 0>, rot=<0 0 0.034 -0.9994>}
 spine_02 {parent=spine_01, loc=<7.2518 -0.035 0>, rot=<0 0 0.0238 -0.9997>}
 spine_03 {parent=spine_02, loc=<6.0104 0 0>, rot=<0 0 0.0149 -0.9999>}
 spine_04 {parent=spine_03, loc=<8.1024 0 0>, rot=<0 0 -0.0132 -0.9999>}
 spine_05 {parent=spine_04, loc=<10.0146 0 0>, rot=<0 0 -0.0015 -1>}
 neck_01 {parent=spine_05, loc=<11.3138 0.1529 0>, rot=<0 0 -0.2003 -0.9797>}
 neck_02 {parent=neck_01, loc=<4.5627 0 0>, rot=<0 0 0.0504 -0.9987>}
 head {parent=neck_02, loc=<5.3914 0 0>, rot=<0 0 0.115 -0.9934>}
 attach_cape {parent=spine_05, loc=<9.017 -11.714 0>, rot=<0.0616 -0.7044 -0.0616 -0.7044>}
 cape {parent=attach_cape, type=Null, loc=<0 2.88 -0.88>, rot=<0 0 0 1>}
 attach_backpack {parent=spine_05, loc=<-0.123 -13.473 0>, rot=<-0.0093 -0.707 0.0093 -0.707>}
 Backpack_BR {parent=attach_backpack, type=Null, loc=<0 4 -0.44>, rot=<0 0 0 1>}
 clavicle_l {parent=spine_05, loc=<4.3575 -0.1466 -2.3223>, rot=<0.0465 -0.7289 -0.009 -0.6829>}
 upperarm_l {parent=clavicle_l, loc=<15.5111 0 0>, rot=<0.0226 -0.3234 -0.0088 -0.9459>}
 lowerarm_l {parent=upperarm_l, loc=<26.6514 0 0>, rot=<0.0391 -0.0099 -0.2453 -0.9686>}
 hand_l {parent=lowerarm_l, loc=<22.3491 0 0>, rot=<-0.7081 -0.0507 0.0975 -0.6975>}
 thumb_01_l {parent=hand_l, loc=<1.3118 -0.3191 -2.459>, rot=<0.6546 -0.4042 0.0664 -0.6354>}
 thumb_02_l {parent=thumb_01_l, loc=<5.9336 -0.0448 0.0401>, rot=<-0.0037 0.0213 0.298 -0.9543>}
 thumb_03_l {parent=thumb_02_l, loc=<2.2223 0.3682 -0.0907>, rot=<0.0118 -0.013 0.0066 -0.9998>}
 index_metacarpal_l {parent=hand_l, loc=<1.6535 0.0183 -2.96>, rot=<0.0158 -0.0124 -0.0173 -0.9996>}
 index_01_l {parent=index_metacarpal_l, loc=<8.3008 0.0357 -0.0524>, rot=<-0.0044 0.0485 0.0924 -0.9945>}
 index_02_l {parent=index_01_l, loc=<3.3587 -0.3004 0.0452>, rot=<-0.002 0.0027 0.3225 -0.9466>}
 index_03_l {parent=index_02_l, loc=<2.5578 -0.123 -0.0252>, rot=<-0.002 0.0067 0.2558 -0.9667>}
 middle_metacarpal_l {parent=hand_l, loc=<2.1856 -0.1725 -0.5984>, rot=<0.006 0.0168 -0.0329 -0.9993>}
 middle_01_l {parent=middle_metacarpal_l, loc=<7.3275 -0.2447 0.0129>, rot=<-0.0208 0.0042 0.0858 -0.9961>}
 middle_02_l {parent=middle_01_l, loc=<4.0123 0.1401 0.1175>, rot=<-0.0218 0.0258 0.3512 -0.9357>}
 middle_03_l {parent=middle_02_l, loc=<2.6077 0.0534 0.1373>, rot=<0.0082 0.0188 0.1708 -0.9851>}
 ring_metacarpal_l {parent=hand_l, loc=<2.2074 -0.3625 1.3638>, rot=<-0.0284 0.0722 -0.0458 -0.9959>}
 ring_01_l {parent=ring_metacarpal_l, loc=<7.0295 -0.6197 -0.1446>, rot=<-0.0489 -0.0252 0.1892 -0.9804>}
 ring_02_l {parent=ring_01_l, loc=<3.6174 0.4647 -0.1027>, rot=<0 0 0.2806 -0.9598>}
 ring_03_l {parent=ring_02_l, loc=<2.4152 0.114 0.0199>, rot=<0 0 0.1189 -0.9929>}
 pinky_metacarpal_l {parent=hand_l, loc=<1.7616 -0.9753 3.3814>, rot=<-0.0397 0.0814 -0.069 -0.9935>}
 pinky_01_l {parent=pinky_metacarpal_l, loc=<7.1709 -0.4831 -0.1094>, rot=<-0.0654 -0.0035 0.2242 -0.9723>}
 pinky_02_l {parent=pinky_01_l, loc=<2.8274 0.0893 -0.1043>, rot=<0.0037 0.0011 0.2854 -0.9584>}
 pinky_03_l {parent=pinky_02_l, loc=<2.2046 0.0307 -0.0078>, rot=<-0.0002 0.0098 0.1445 -0.9895>}
 deform_wrist_up_base_l {parent=hand_l, loc=<0 2.2682 0>, rot=<0 0 0 -1>}
 deform_wrist_up_l {parent=deform_wrist_up_base_l, loc=<0 0 0>, rot=<0 0 0 1>}
 deform_wrist_dn_base_l {parent=hand_l, loc=<0 -2.2342 0>, rot=<0 0 0 -1>}
 deform_wrist_dn_l {parent=deform_wrist_dn_base_l, loc=<0 0 0>, rot=<0 0 0 1>}
 lowerarm_twist_02_l {parent=lowerarm_l, loc=<9.6199 0 0>, rot=<0 0 0 -1>}
 lowerarm_twist_01_l {parent=lowerarm_l, loc=<17.8549 0 0>, rot=<0 0 0 -1>}
 deform_elbow_in_base_l {parent=lowerarm_l, loc=<0.2112 5.3237 -0.0001>, rot=<0.078 0 0 -0.997>}
 deform_elbow_in_l {parent=deform_elbow_in_base_l, loc=<0 0 0>, rot=<0 0 0 1>}
 deform_elbow_out_base_l {parent=lowerarm_l, loc=<0 -3.9514 0.6217>, rot=<0.078 0 0 -0.997>}
 deform_elbow_out_l {parent=deform_elbow_out_base_l, loc=<0 0 0>, rot=<0 0 0 1>}
 upperarm_twist_01_l {parent=upperarm_l, loc=<0.0041 0 0>, rot=<0 0 0 -1>}
 upperarm_twist_02_l {parent=upperarm_l, loc=<13.7941 0 0>, rot=<0 0 0 -1>}
 deform_upperarm_up_base_l {parent=upperarm_l, loc=<-4.6879 -1.0812 1.2799>, rot=<0.1089 0.0365 -0.0071 -0.9934>}
 deform_upperarm_up_l {parent=deform_upperarm_up_base_l, loc=<0 0 0>, rot=<0 0 0 1>}
 deform_upperarm_fr_base_l {parent=upperarm_l, loc=<-1.4429 5.6609 -6.2291>, rot=<0.1089 0.0365 -0.0071 -0.9934>}
 deform_upperarm_fr_l {parent=deform_upperarm_fr_base_l, loc=<0 0 0>, rot=<0 0 0 1>}
 deform_upperarm_bk_base_l {parent=upperarm_l, loc=<0.0311 -8.0603 -2.234>, rot=<0.1089 0.0365 -0.0071 -0.9934>}
 deform_upperarm_bk_l {parent=deform_upperarm_bk_base_l, loc=<0 0 0>, rot=<0 0 0 1>}
 upperRetArm_l {parent=clavicle_l, loc=<13.338 -0.7151 -0.7834>, rot=<0.1228 -0.286 -0.0515 -0.9489>}
 lowerRetArm_l {parent=upperRetArm_l, loc=<27.9684 0 0>, rot=<0 0 -0.24 -0.9708>}
 lowerRetArm_twist_01_l {parent=lowerRetArm_l, loc=<5.7119 0 0>, rot=<0 0 0 1>}
 lowerRetArm_twist_02_l {parent=lowerRetArm_l, loc=<11.4238 0 0>, rot=<0 0 0 1>}
 lowerRetArm_twist_03_l {parent=lowerRetArm_l, loc=<17.1357 0 0>, rot=<0 0 0 1>}
 lowerRetArm_twist_04_l {parent=lowerRetArm_l, loc=<21.3491 0 0>, rot=<0 0 0 1>}
 handRet_l {parent=lowerRetArm_l, loc=<22.8476 0 0>, rot=<0 0 0 1>}
 upperRetArm_twist_01_l {parent=upperRetArm_l, loc=<1 0 0>, rot=<0 0 0 1>}
 upperRetArm_twist_02_l {parent=upperRetArm_l, loc=<9.3228 0 0>, rot=<0 0 0 1>}
 upperRetArm_twist_03_l {parent=upperRetArm_l, loc=<18.6456 0 0>, rot=<0 0 0 1>}
 clavicle_r {parent=spine_05, loc=<4.3575 -0.1466 2.3223>, rot=<0.7289 0.0465 -0.6829 0.009>}
 upperarm_r {parent=clavicle_r, loc=<-15.5111 0 0>, rot=<0.0226 -0.3234 -0.0088 -0.9459>}
 lowerarm_r {parent=upperarm_r, loc=<-26.6514 0 0>, rot=<0.0391 -0.0099 -0.2453 -0.9686>}
 hand_r {parent=lowerarm_r, loc=<-22.3491 0 0>, rot=<-0.7081 -0.0507 0.0975 -0.6975>}
 thumb_01_r {parent=hand_r, loc=<-1.3118 0.3191 2.459>, rot=<0.6546 -0.4042 0.0664 -0.6354>}
 thumb_02_r {parent=thumb_01_r, loc=<-5.9336 0.0448 -0.0401>, rot=<-0.0037 0.0213 0.298 -0.9543>}
 thumb_03_r {parent=thumb_02_r, loc=<-2.2223 -0.3682 0.0907>, rot=<0.0118 -0.013 0.0066 -0.9998>}
 index_metacarpal_r {parent=hand_r, loc=<-1.6535 -0.0183 2.96>, rot=<0.0158 -0.0124 -0.0173 -0.9996>}
 index_01_r {parent=index_metacarpal_r, loc=<-8.3008 -0.0357 0.0524>, rot=<-0.0044 0.0485 0.0924 -0.9945>}
 index_02_r {parent=index_01_r, loc=<-3.3587 0.3004 -0.0452>, rot=<-0.002 0.0027 0.3225 -0.9466>}
 index_03_r {parent=index_02_r, loc=<-2.5578 0.123 0.0252>, rot=<-0.002 0.0067 0.2558 -0.9667>}
 middle_metacarpal_r {parent=hand_r, loc=<-2.1856 0.1725 0.5984>, rot=<0.006 0.0168 -0.0329 -0.9993>}
 middle_01_r {parent=middle_metacarpal_r, loc=<-7.3275 0.2447 -0.0129>, rot=<-0.0208 0.0042 0.0858 -0.9961>}
 middle_02_r {parent=middle_01_r, loc=<-4.0123 -0.1401 -0.1175>, rot=<-0.0218 0.0258 0.3512 -0.9357>}
 middle_03_r {parent=middle_02_r, loc=<-2.6077 -0.0534 -0.1373>, rot=<0.0082 0.0188 0.1708 -0.9851>}
 ring_metacarpal_r {parent=hand_r, loc=<-2.2074 0.3625 -1.3638>, rot=<-0.0284 0.0722 -0.0458 -0.9959>}
 ring_01_r {parent=ring_metacarpal_r, loc=<-7.0295 0.6197 0.1446>, rot=<-0.0489 -0.0252 0.1892 -0.9804>}
 ring_02_r {parent=ring_01_r, loc=<-3.6174 -0.4647 0.1027>, rot=<0 0 0.2806 -0.9598>}
 ring_03_r {parent=ring_02_r, loc=<-2.4152 -0.114 -0.0199>, rot=<0 0 0.1189 -0.9929>}
 pinky_metacarpal_r {parent=hand_r, loc=<-1.7616 0.9753 -3.3814>, rot=<-0.0397 0.0814 -0.069 -0.9935>}
 pinky_01_r {parent=pinky_metacarpal_r, loc=<-7.1709 0.4831 0.1094>, rot=<-0.0654 -0.0035 0.2242 -0.9723>}
 pinky_02_r {parent=pinky_01_r, loc=<-2.8274 -0.0893 0.1043>, rot=<0.0037 0.0011 0.2854 -0.9584>}
 pinky_03_r {parent=pinky_02_r, loc=<-2.2046 -0.0307 0.0078>, rot=<-0.0002 0.0098 0.1445 -0.9895>}
 deform_wrist_dn_base_r {parent=hand_r, loc=<-0.0004 2.2345 -0.0001>, rot=<0 0 0 -1>}
 deform_wrist_dn_r {parent=deform_wrist_dn_base_r, loc=<0 0 0>, rot=<0 0 0 1>}
 deform_wrist_up_base_r {parent=hand_r, loc=<0.0003 -2.2684 0.0001>, rot=<0 0 0 -1>}
 deform_wrist_up_r {parent=deform_wrist_up_base_r, loc=<0 0 0>, rot=<0 0 0 1>}
 lowerarm_twist_02_r {parent=lowerarm_r, loc=<-9.6199 0 0>, rot=<0 0 0 -1>}
 lowerarm_twist_01_r {parent=lowerarm_r, loc=<-17.8549 0 0>, rot=<0 0 0 -1>}
 deform_elbow_in_base_r {parent=lowerarm_r, loc=<-0.2112 -5.3237 0.0002>, rot=<0.078 0 0 -0.997>}
 deform_elbow_in_r {parent=deform_elbow_in_base_r, loc=<-0.0001 0 0.0002>, rot=<0 0 0 -1>}
 deform_elbow_out_base_r {parent=lowerarm_r, loc=<0.0003 3.9513 -0.622>, rot=<0.078 0 0 -0.997>}
 deform_elbow_out_r {parent=deform_elbow_out_base_r, loc=<0 0 0>, rot=<0 0 0 -1>}
 upperarm_twist_01_r {parent=upperarm_r, loc=<-0.0041 0 0>, rot=<0 0 0 -1>}
 upperarm_twist_02_r {parent=upperarm_r, loc=<-13.7941 0 0>, rot=<0 0 0 -1>}
 deform_upperarm_up_base_r {parent=upperarm_r, loc=<4.6882 1.0813 -1.2803>, rot=<0.1089 0.0365 -0.0071 -0.9934>}
 deform_upperarm_up_r {parent=deform_upperarm_up_base_r, loc=<0 0 0>, rot=<0 0 0 -1>}
 deform_upperarm_fr_base_r {parent=upperarm_r, loc=<1.4429 -5.6609 6.229>, rot=<0.1089 0.0365 -0.0071 -0.9934>}
 deform_upperarm_fr_r {parent=deform_upperarm_fr_base_r, loc=<0 0 0>, rot=<0 0 0 1>}
 deform_upperarm_bk_base_r {parent=upperarm_r, loc=<-0.0311 8.0602 2.234>, rot=<0.1089 0.0365 -0.0071 -0.9934>}
 deform_upperarm_bk_r {parent=deform_upperarm_bk_base_r, loc=<0 0 0>, rot=<0 0 0 1>}
 upperRetArm_r {parent=clavicle_r, loc=<-13.338 0.7151 0.7837>, rot=<0.1228 -0.286 -0.0515 -0.9489>}
 lowerRetArm_r {parent=upperRetArm_r, loc=<-27.9684 0 -0.0001>, rot=<0 0 -0.24 -0.9708>}
 lowerRetArm_twist_01_r {parent=lowerRetArm_r, loc=<-5.7116 0 -0.0005>, rot=<0 0 0 1>}
 lowerRetArm_twist_02_r {parent=lowerRetArm_r, loc=<-11.4237 0 -0.0001>, rot=<0 0 0 1>}
 lowerRetArm_twist_03_r {parent=lowerRetArm_r, loc=<-17.1353 0 -0.0006>, rot=<0 0 0 1>}
 lowerRetArm_twist_04_r {parent=lowerRetArm_r, loc=<-21.8475 0 0>, rot=<0 0 0 1>}
 handRet_r {parent=lowerRetArm_r, loc=<-22.8475 0 -0.0002>, rot=<0 0 0 1>}
 upperRetArm_twist_01_r {parent=upperRetArm_r, loc=<-1 0 0>, rot=<0 0 0 1>}
 upperRetArm_twist_02_r {parent=upperRetArm_r, loc=<-9.3224 0.0001 -0.0006>, rot=<0 0 0 1>}
 upperRetArm_twist_03_r {parent=upperRetArm_r, loc=<-18.6454 0.0001 -0.0003>, rot=<0 0 0 1>}
 deform_pec_aim_base_l {parent=spine_04, loc=<7.8732 12.6388 -2.4123>, rot=<0.0102 -0.4602 0.3979 -0.7936>}
 deform_pec_aim_l {parent=deform_pec_aim_base_l, loc=<0 0 0>, rot=<0 0 0 1>}
 deform_pec_base_l {parent=deform_pec_aim_l, loc=<10.766 0 0>, rot=<0 0 0 1>}
 deform_pec_l {parent=deform_pec_base_l, loc=<0.055 0 0>, rot=<0 0 0 1>}
 deform_pec_aim_base_r {parent=spine_04, loc=<7.8727 12.6388 2.4123>, rot=<-0.4602 -0.0102 0.7936 0.3979>}
 deform_pec_aim_r {parent=deform_pec_aim_base_r, loc=<0 0 0>, rot=<0 0 0 1>}
 deform_pec_base_r {parent=deform_pec_aim_r, loc=<-10.7662 -0.0006 -0.0007>, rot=<0 0 0 1>}
 deform_pec_r {parent=deform_pec_base_r, loc=<0 0 0>, rot=<0 0 0 1>}
 deform_lat_l_base {parent=spine_04, loc=<1.3126 0.4828 -13.4832>, rot=<0.0294 -0.7065 -0.0294 -0.7065>}
 deform_lat_l {parent=deform_lat_l_base, loc=<0 0 0>, rot=<0 0 0 1>}
 deform_lat_r_base {parent=spine_04, loc=<1.3124 0.4828 13.4832>, rot=<0.7065 0.0294 0.7065 -0.0294>}
 deform_lat_r {parent=deform_lat_r_base, loc=<0 0 0>, rot=<0 0 0 1>}
 thigh_l {parent=pelvis, loc=<-2.5693 -0.0109 -7.7501>, rot=<0.0012 0.0572 0.0155 -0.9982>}
 calf_l {parent=thigh_l, loc=<-39.0484 0 0>, rot=<0.0006 -0.0161 -0.0391 -0.9991>}
 foot_l {parent=calf_l, loc=<-42.561 0.1077 0.4292>, rot=<0.0418 -0.0455 0.0343 -0.9975>}
 ball_l {parent=foot_l, loc=<-6.1599 13.8264 0.6806>, rot=<0.0001 -0.0001 -0.7063 -0.7079>}
 calf_twist_02_l {parent=calf_l, loc=<-18.8064 -0.0967 0.0397>, rot=<0.0237 0.0095 -0.0035 -0.9997>}
 calf_twist_01_l {parent=calf_l, loc=<-35.813 -0.0967 0.0397>, rot=<0.0237 0.0095 -0.0035 -0.9997>}
 deform_knee_in_base_l {parent=calf_l, loc=<1.298 -5.2957 -0.0496>, rot=<0 0 0 -1>}
 deform_knee_in_l {parent=deform_knee_in_base_l, loc=<0 0 0>, rot=<0 0 0 1>}
 deform_knee_out_base_l {parent=calf_l, loc=<0 4 0>, rot=<0 0 0 -1>}
 deform_knee_out_l {parent=deform_knee_out_base_l, loc=<0 0 0>, rot=<0 0 0 1>}
 thigh_twist_01_l {parent=thigh_l, loc=<-21.6303 0 0>, rot=<0.0013 0 0 -1>}
 deform_groin_base_l {parent=thigh_l, loc=<-2.1908 8.1142 -1.7442>, rot=<0 0 0 -1>}
 deform_groin_l {parent=deform_groin_base_l, loc=<0 0 0>, rot=<0 0 0 1>}
 thigh_r {parent=pelvis, loc=<-2.5693 -0.0109 7.7501>, rot=<0.0572 -0.0012 0.9982 0.0155>}
 calf_r {parent=thigh_r, loc=<39.0484 0 0>, rot=<0.0006 -0.0161 -0.0391 -0.9991>}
 foot_r {parent=calf_r, loc=<42.561 -0.1077 -0.4292>, rot=<0.0418 -0.0455 0.0343 -0.9975>}
 ball_r {parent=foot_r, loc=<6.1599 -13.8264 -0.6806>, rot=<0.0001 -0.0001 -0.7063 -0.7079>}
 calf_twist_02_r {parent=calf_r, loc=<18.8064 0.0967 -0.0397>, rot=<0.0237 0.0095 -0.0035 -0.9997>}
 calf_twist_01_r {parent=calf_r, loc=<35.813 0.0967 -0.0397>, rot=<0.0237 0.0095 -0.0035 -0.9997>}
 deform_knee_in_base_r {parent=calf_r, loc=<-1.298 5.2957 0.0496>, rot=<0 0 0 -1>}
 deform_knee_in_r {parent=deform_knee_in_base_r, loc=<0 0 0>, rot=<0 0 0 1>}
 deform_knee_out_base_r {parent=calf_r, loc=<0 -4 0>, rot=<0 0 0 -1>}
 deform_knee_out_r {parent=deform_knee_out_base_r, loc=<0 0 0>, rot=<0 0 0 1>}
 thigh_twist_01_r {parent=thigh_r, loc=<21.6303 0 0>, rot=<0.0013 0 0 -1>}
 deform_groin_base_r {parent=thigh_r, loc=<2.1908 -8.1142 1.7442>, rot=<0 0 0 -1>}
 deform_groin_r {parent=deform_groin_base_r, loc=<0 0 0>, rot=<0 0 0 1>}
 deform_glute_base_l {parent=pelvis, loc=<-8.8064 -8.8222 -8.4287>, rot=<0.0275 0.0501 -0.4472 -0.8926>}
 deform_glute_l {parent=deform_glute_base_l, loc=<0 0 0>, rot=<0 0 0 1>}
 deform_glute_base_r {parent=pelvis, loc=<-8.8062 -8.8222 8.4287>, rot=<0.0501 -0.0275 0.8926 -0.4472>}
 deform_glute_r {parent=deform_glute_base_r, loc=<0 0 0>, rot=<0 0 0 1>}
]
